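# Editor utility widget for one screen of the Multi Cam Viewer animation tool, showing a camera view with a view selector and label.
# On construct it applies its material to the image and fills the combo box from its view list, restoring the current option.
# View changes update the label and broadcast ViewChanged with the screen ID.
# The parent widget's Hide Combo Boxes dispatcher toggles the combo box's visibility.
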EditorUtilityWidgetBlueprint /Game/NeverCook/Animation/MultiCamViewer/Components/EUW_MultiCamViewer_Screen
graphs [
 EventGraph {nodes [
   n1 = "On Selection Changed (ComboBox_Views)"!() -> (OutputDelegate, then, SelectedItem:String, SelectionType:ESelectInfo Enum) {is=title K2Node_ComponentBoundEvent}
   > ChangeView!(execute, self, Option:String=@n1.SelectedItem) -> (then)
   LoadAvalibleViews() {is=title K2Node_CustomEvent}
   n2 = Get ComboBox_Views() -> (ComboBox_Views:Editor Utility Combo Box String Object Reference)
   > ClearOptions(self:"Combo Box (String) Object Reference"=@n2)
   > n3 = For Each Loop(Array=Get AvalibleViews() -> (AvalibleViews:Array of Strings))
   ChangeViewSelection!(execute=@n3.Completed, self, Option:String=Get CurrentOption() -> (CurrentOption:String)) -> (then)
   AddOption!(execute=@n3.LoopBody, self:"Combo Box (String) Object Reference"=@n2, Option:String=@n3) -> (then)
   n4 = ChangeButtonsVisibility!() -> (OutputDelegate, then, IsVisible:bool=~) {is=title K2Node_CustomEvent}
   > SetVisibility(self=Get ComboBox_Views(), InVisibility=Select(Option 0=Collapsed, Option 1=Visible, Index=@n4.IsVisible))
   n5 = SetImageAsTexture!() -> (OutputDelegate, then, Texture:Texture 2D Object Reference) {is=title K2Node_CustomEvent}
   n6 = Get Image_View() -> (Image_View:Image Object Reference)
   > SetBrushFromTexture!(execute, self:Image Object Reference=@n6, Texture:Texture 2D Object Reference=@n5.Texture, bMatchSize:bool=false) -> (then)
   n7 = SetImageAsMaterial!() -> (OutputDelegate, then, Material:Material Interface Object Reference) {is=title K2Node_CustomEvent}
   > SetBrushFromMaterial(self=@n6, Material=@n7.Material)
   n8 = ChangeViewSelection!() -> (OutputDelegate, then, Option:String) {is=title K2Node_CustomEvent}
   > SetSelectedOption!(execute, self:"Combo Box (String) Object Reference"=Get ComboBox_Views(), Option:String=@n8.Option) -> (then)
   > ChangeView(Option=@n8.Option)
   n9 = ChangeView!() -> (OutputDelegate, then, Option:String) {is=title K2Node_CustomEvent}
   > n10 = Set CurrentOption(CurrentOption:String=@n9.Option) -> (Output_Get:String)
   > Call View Changed!(execute:Exec, self:EUW Multi Cam Viewer Screen Object Reference, Option:String=@n10, ScreenId:String=Get ScreenId() -> (ScreenId:String)) -> (then:Exec) {is=title K2Node_CallDelegate}
   > "SetText (Text)"(self=Get TextBlock_CurrentView() -> (TextBlock_CurrentView:Text Object Reference), InText="To Text (String)"(InString=@n10))
   BindScreenBoxVisibility() {is=title K2Node_CustomEvent}
   > Bind Event to Hide Combo Boxes(self:EUW Multi Cam Viewer Object Reference=Get ParentWidget() -> (ParentWidget:EUW Multi Cam Viewer Object Reference), Delegate=@n4) {is=title K2Node_AddDelegate}
   Event Construct()
   > SetImageAsMaterial!(execute, self, Material:Material Interface Object Reference=Get Material() -> (Material:Material Interface Object Reference)) -> (then) {is=title K2Node_CallFunction}
   > LoadAvalibleViews() {is=title K2Node_CallFunction}
   > BindScreenBoxVisibility() {is=title K2Node_CallFunction}
  ]}
 ViewChanged {nodes [ViewChanged!() -> (then, Option:String, ScreenId:String) {is=title K2Node_FunctionEntry}]}
]
variables [
 AvalibleViews {type="array<string>", default [Default_01,Default_02]}
 CurrentOption {type=string, default=3/4 L}
 Material {type=MaterialInterface, default=null}
 ParentWidget {type=EUW_MultiCamViewer_C, default=null}
 ScreenId {type=string, default=""}
 ViewChanged {default=()}
]
widgets {
 info {rootWidgetClass=&CanvasPanel, widgetCount=5}
 widgets [
  CanvasPanel_61 {widgetClassPath=&CanvasPanel, widget=CanvasPanel_61}
  Border {
   widgetClassPath=&Border
   parent=CanvasPanel_61
   widget=Border
   properties {
    background {
     bIsDynamicallyLoaded=false
     drawAs=Box
     imageType=NoImage
     mirroring=NoMirror
     resourceName=None
     resourceObject=null
     tiling=NoTile
     imageSize=<32 32>
     margin=<ltrb 0 0 0 0>
     outlineSettings {
      bUseBrushTransparency=false
      roundingType=HalfHeightRadius
      width=0
      color {colorUseRule=UseColor_Specified, specifiedColor=<rgba 0 0 0 0>}
      cornerRadii=<0 0 0 0>
     }
     tintColor {colorUseRule=UseColor_Specified, specifiedColor=<rgba 0 0 0 1>}
     uVRegion {bIsValid=false, max=<0 0>, min=<0 0>}
    }
    padding=<ltrb 1.25 1.25 1.25 1.25>
   }
   slot {
    class=CanvasPanelSlot
    properties {layoutData {alignment=<0.5 0.5>, anchors {maximum=<0.5 0.5>, minimum=<0.5 0.5>}, offsets=<ltrb 0 0 1920 1080>}}
   }
  }
  Image_View {
   widgetClassPath=&Image
   bIsVariable=true
   parent=Border
   widget=Image_View
   properties {brush {
     bIsDynamicallyLoaded=false
     drawAs=Image
     imageType=FullColor
     mirroring=NoMirror
     resourceName=None
     resourceObject=&/RT_MultiCamViewer_Front
     tiling=NoTile
     imageSize=<32 32>
     margin=<ltrb 0 0 0 0>
     outlineSettings {
      bUseBrushTransparency=false
      roundingType=HalfHeightRadius
      width=0
      color {colorUseRule=UseColor_Specified, specifiedColor=<rgba 0 0 0 0>}
      cornerRadii=<0 0 0 0>
     }
     tintColor {colorUseRule=UseColor_Specified, specifiedColor=<rgba 1 1 1 1>}
     uVRegion {bIsValid=false, max=<0 0>, min=<0 0>}
    }}
   slot {class=BorderSlot, properties {padding=<ltrb 1.25 1.25 1.25 1.25>}}
  }
  ComboBox_Views {
   widgetClassPath=/Script/Blutility.EditorUtilityComboBoxString
   bIsVariable=true
   parent=CanvasPanel_61
   widget=ComboBox_Views
   properties {
    selectedOption=Default_01
    font {
     bForceMonospaced=false
     bMaterialIsStencil=false
     fontMaterial=null
     fontObject=&Roboto
     letterSpacing=0
     monospacedWidth=1
     size=22.5
     skewAmount=0
     typefaceFontName=Bold
     outlineSettings {
      bApplyOutlineToDropShadows=false
      bMiteredCorners=false
      bSeparateFillAlpha=false
      outlineMaterial=null
      outlineSize=0
      outlineColor=<rgba 0 0 0 1>
     }
    }
    foregroundColor {colorUseRule=UseColor_Specified, specifiedColor=<rgba 0.5271 0.5271 0.5271 1>}
   }
   slot {
    class=CanvasPanelSlot
    properties {zOrder=3, layoutData {alignment=<0.5 1.25>, anchors {maximum=<0.5 1>, minimum=<0.5 1>}, offsets=<ltrb 0 0 720 75>}}
   }
  }
  TextBlock_CurrentView {
   widgetClassPath=&TextBlock
   bIsVariable=true
   parent=CanvasPanel_61
   widget=TextBlock_CurrentView
   properties {
    justification=Center
    text=Test
    font {
     bForceMonospaced=false
     bMaterialIsStencil=false
     fontMaterial=null
     fontObject=&NotoSans
     letterSpacing=0
     monospacedWidth=1
     size=15
     skewAmount=0
     typefaceFontName=NotoSans-Regular
     outlineSettings {
      bApplyOutlineToDropShadows=false
      bMiteredCorners=false
      bSeparateFillAlpha=false
      outlineMaterial=null
      outlineSize=0
      outlineColor=<rgba 0 0 0 1>
     }
    }
   }
   slot {
    class=CanvasPanelSlot
    properties {bAutoSize=true, zOrder=2, layoutData {alignment=<0.5 2>, anchors {maximum=<0.5 1>, minimum=<0.5 1>}, offsets=<ltrb 0 0 100 30>}}
   }
  }
 ]
}
