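# Editor utility widget row that edits one text overlay of the MultiCamViewer tool.
# PreConstruct binds single-property views to the widget's own text, size, position, color, outline and visibility variables.
# Each change is applied directly to the referenced TextBlock, then sent through PassProperty to the parent viewer's SaveTextEdit with the row Index.
# The delete button asks the parent to remove that overlay by Index.
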EditorUtilityWidgetBlueprint /Game/NeverCook/Animation/MultiCamViewer/Components/EUW_MultiCamViewer_TextDetails
graphs [EventGraph {nodes [
   n1 = "On Property Changed (PropertyView_Position)"!() -> (OutputDelegate, then, PropertyName:Name=None) {is=title K2Node_ComponentBoundEvent}
   > n2 = Cast To CanvasPanelSlot!(
    execute
    Object=Get Slot!(self:Widget Object Reference=Get TextComponentRef() -> (TextComponentRef:Text Object Reference)) -> (Slot:Panel Slot Object Reference)
   ) -> (then, CastFailed, AsCanvas Panel Slot:Canvas Panel Slot Object Reference) {is=title K2Node_DynamicCast}
   > SetPosition!(execute, self:Canvas Panel Slot Object Reference=@n2, InPosition:Vector 2D Structure=Get Position() -> (Position:Vector 2D Structure)) -> (then)
   > PassProperty!(execute, self, Property:Name=@n1.PropertyName) -> (then)
   n3 = "On Property Changed (PropertyView_Size)"!() -> (OutputDelegate, then, PropertyName:Name=None) {is=title K2Node_ComponentBoundEvent}
   n4 = Get TextComponentRef()
   > n5 = Set members in Slate Font Info!(
    execute:Exec
    StructRef:"Slate Font Info Structure (by ref)"=Get Font!(self:Text Object Reference=@n4) -> (Font:Slate Font Info Structure)
    Size:float=Get Size() -> (Size:double)
   ) -> (then:Exec, StructOut:"Slate Font Info Structure (by ref)") {is=title K2Node_SetFieldsInStruct}
   > SetFont(self=@n4, InFontInfo=@n5)
   > PassProperty(Property=@n3.PropertyName)
   n6 = "On Property Changed (PropertyView_Text)"!() -> (OutputDelegate, then, PropertyName:Name=None) {is=title K2Node_ComponentBoundEvent}
   > "SetText (Text)"(self=Get TextComponentRef(), InText="To Text (String)"(InString=Get Text!() -> (Text:String)))
   > PassProperty(Property=@n6.PropertyName)
   n7 = "On Property Changed (PropertyView_Color)"!() -> (OutputDelegate, then, PropertyName:Name=None) {is=title K2Node_ComponentBoundEvent}
   n8 = Get TextComponentRef()
   > n9 = Set members in Slate Color!(
    execute:Exec
    StructRef:"Slate Color Structure (by ref)"=Get ColorAndOpacity!(self:Text Object Reference=@n8) -> (ColorAndOpacity:Slate Color Structure)
    SpecifiedColor:Linear Color Structure=Get Color() -> (Color:Linear Color Structure)
   ) -> (then:Exec, StructOut:"Slate Color Structure (by ref)") {is=title K2Node_SetFieldsInStruct}
   > SetColorAndOpacity(self=@n8, InColorAndOpacity=@n9)
   > PassProperty(Property=@n7.PropertyName)
   n10 = "On Property Changed (PropertyView_Visible)"!() -> (OutputDelegate, then, PropertyName:Name=None) {is=title K2Node_ComponentBoundEvent}
   > SetVisibility(
    self=Get TextComponentRef()
    InVisibility=Select(Option 0=Collapsed, Option 1=SelfHitTestInvisible, Index=Get TextVisible() -> (TextVisible:bool))
   )
   > PassProperty(Property=@n10.PropertyName)
   "On Clicked (ButtonDelete)"() {is=title K2Node_ComponentBoundEvent}
   > Delete single text overlay!(
    execute
    self:EUW Multi Cam Viewer Object Reference=Get Parent() -> (Parent:EUW Multi Cam Viewer Object Reference)
    IndexToRemove:int=Get Index() -> (Index:int=0)
   ) -> (then)
   n11 = "On Property Changed (PropertyView_OutlineSize)"!() -> (OutputDelegate, then, PropertyName:Name=None) {is=title K2Node_ComponentBoundEvent}
   n12 = Get TextComponentRef()
   n13 = Get Font(self=@n12)
   > n14 = Set members in Font Outline Settings!(
    execute:Exec
    StructRef:"Font Outline Settings Structure (by ref)"=Break Slate Font Info!(SlateFontInfo:"Slate Font Info Structure (by ref)"=@n13) -> (OutlineSettings:Font Outline Settings Structure) {is=title K2Node_BreakStruct}
    OutlineSize:int=Get OutlineSize() -> (OutlineSize:int=0)
   ) -> (then:Exec, StructOut:"Font Outline Settings Structure (by ref)") {is=title K2Node_SetFieldsInStruct}
   > n15 = Set members in Slate Font Info!(execute, StructRef=@n13, OutlineSettings:Font Outline Settings Structure=@n14) -> (then, StructOut)
   > SetFont(self=@n12, InFontInfo=@n15)
   > PassProperty(Property=@n11.PropertyName)
   n16 = "On Property Changed (PropertyView_OutlineColor)"!() -> (OutputDelegate, then, PropertyName:Name=None) {is=title K2Node_ComponentBoundEvent}
   n17 = Get TextComponentRef()
   n18 = Get Font(self=@n17)
   > n19 = Set members in Font Outline Settings!(
    execute
    StructRef=Break Slate Font Info(SlateFontInfo=@n18)
    OutlineColor:Linear Color Structure=Get OutlineColor() -> (OutlineColor:Linear Color Structure)
   ) -> (then, StructOut)
   > n20 = Set members in Slate Font Info(StructRef=@n18, OutlineSettings=@n19)
   > SetFont(self=@n17, InFontInfo=@n20)
   > PassProperty(Property=@n16.PropertyName)
   n21 = PassProperty!() -> (OutputDelegate, then, Property:Name) {is=title K2Node_CustomEvent}
   > SaveTextEdit!(
    execute
    self:EUW Multi Cam Viewer Object Reference=Get Parent()
    Index:int=Get Index() -> (Index=0)
    Property:Name=@n21.Property
    Text:String=Get Text()
    Size:double=Get Size()
    Position:Vector 2D Structure=Get Position()
    Color:Linear Color Structure=Get Color()
    Visibility:bool=Get TextVisible()
    OutlineSize:int=Get OutlineSize() -> (OutlineSize=0)
    OutlineColor:Linear Color Structure=Get OutlineColor()
   ) -> (then)
   Event PreConstruct()
   > SetObject(
    self=[
     Get PropertyView_Size() -> (PropertyView_Size:Single Property View Object Reference)
     Get PropertyView_Visible() -> (PropertyView_Visible:Single Property View Object Reference)
     Get PropertyView_Position() -> (PropertyView_Position:Single Property View Object Reference)
     Get PropertyView_Color() -> (PropertyView_Color:Single Property View Object Reference)
     Get PropertyView_OutlineSize() -> (PropertyView_OutlineSize:Single Property View Object Reference)
     Get PropertyView_OutlineColor() -> (PropertyView_OutlineColor:Single Property View Object Reference)
     Get PropertyView_Text() -> (PropertyView_Text:Single Property View Object Reference)
    ]
    NewObject=Self-Reference()
   )
  ]}]
variables [
 Color {type=LinearColor, default=<rgba 0 0 0 1>}
 Index {type=integer}
 OutlineColor {type=LinearColor, default=<rgba 1 1 1 1>}
 OutlineSize {type=integer, default=2}
 Parent {type=EUW_MultiCamViewer_C, default=null}
 Position {type=Vector2D, default=<0 0>}
 Size {type=number, default=50}
 Text {type=string, default=New Text}
 TextComponentRef {type=TextBlock, default=null}
 TextVisible {type=boolean, default=true}
]
widgets {
 info {rootWidgetClass=&HorizontalBox, widgetCount=11}
 widgets [
  HorizontalBox_14 {widgetClassPath=&HorizontalBox, widget=HorizontalBox_14}
  SizeBox_0 {
   widgetClassPath=&SizeBox
   parent=HorizontalBox_14
   widget=SizeBox_0
   properties {bOverride_WidthOverride=true, widthOverride=300}
   slot {class=HorizontalBoxSlot, properties {horizontalAlignment=HAlign_Left, size {sizeRule=Fill, value=1}}}
  }
  PropertyView_Text {
   widgetClassPath=&SinglePropertyView
   parent=SizeBox_0
   widget=PropertyView_Text
   properties {propertyName=Text}
   slot {class=SizeBoxSlot}
  }
  PropertyView_Size {
   widgetClassPath=&SinglePropertyView
   parent=HorizontalBox_14
   widget=PropertyView_Size
   properties {propertyName=Size}
   slot {class=HorizontalBoxSlot, properties {horizontalAlignment=HAlign_Right}}
  }
  PropertyView_Position {
   widgetClassPath=&SinglePropertyView
   parent=HorizontalBox_14
   widget=PropertyView_Position
   properties {propertyName=Position}
   slot {class=HorizontalBoxSlot, properties {horizontalAlignment=HAlign_Right}}
  }
  PropertyView_Color {
   widgetClassPath=&SinglePropertyView
   parent=HorizontalBox_14
   widget=PropertyView_Color
   properties {propertyName=Color}
   slot {class=HorizontalBoxSlot}
  }
  PropertyView_OutlineSize {
   widgetClassPath=&SinglePropertyView
   parent=HorizontalBox_14
   widget=PropertyView_OutlineSize
   properties {propertyName=OutlineSize}
   slot {class=HorizontalBoxSlot}
  }
  PropertyView_OutlineColor {
   widgetClassPath=&SinglePropertyView
   parent=HorizontalBox_14
   widget=PropertyView_OutlineColor
   properties {propertyName=OutlineColor}
   slot {class=HorizontalBoxSlot}
  }
  PropertyView_Visible {
   widgetClassPath=&SinglePropertyView
   parent=HorizontalBox_14
   widget=PropertyView_Visible
   properties {nameOverride=Visible, propertyName=TextVisible}
   slot {class=HorizontalBoxSlot}
  }
  ButtonDelete {
   widgetClassPath=&EditorUtilityButton
   parent=HorizontalBox_14
   widget=ButtonDelete
   properties {widgetStyle {
     clickedSlateSound {resourceObject=null}
     disabled {
      bIsDynamicallyLoaded=false
      drawAs=RoundedBox
      imageType=NoImage
      mirroring=NoMirror
      resourceName=None
      resourceObject=null
      tiling=NoTile
      imageSize=<0 0>
      margin=<ltrb 0 0 0 0>
      outlineSettings {
       bUseBrushTransparency=false
       roundingType=FixedRadius
       width=1
       color {colorUseRule=UseColor_ColorTable, specifiedColor=<rgba 0 0 0 0>}
       cornerRadii=<4 4 4 4>
      }
      tintColor {colorUseRule=UseColor_Specified, specifiedColor=<rgba 0.0395 0.0395 0.0395 1>}
      uVRegion {bIsValid=false, max=<0 0>, min=<0 0>}
     }
     disabledForeground {colorUseRule=UseColor_ColorTable, specifiedColor=<rgba 0 0 0 0>}
     hovered {
      bIsDynamicallyLoaded=false
      drawAs=RoundedBox
      imageType=NoImage
      mirroring=NoMirror
      resourceName=None
      resourceObject=null
      tiling=NoTile
      imageSize=<32 32>
      margin=<ltrb 0 0 0 0>
      outlineSettings {
       bUseBrushTransparency=false
       roundingType=FixedRadius
       width=1
       color {colorUseRule=UseColor_ColorTable, specifiedColor=<rgba 0 0 0 0>}
       cornerRadii=<4 4 4 4>
      }
      tintColor {colorUseRule=UseColor_Specified, specifiedColor=<rgba 0.0953 0.0953 0.0953 1>}
      uVRegion {bIsValid=false, max=<0 0>, min=<0 0>}
     }
     hoveredForeground {colorUseRule=UseColor_ColorTable, specifiedColor=<rgba 0 0 0 0>}
     hoveredSlateSound {resourceObject=null}
     normal {
      bIsDynamicallyLoaded=false
      drawAs=RoundedBox
      imageType=NoImage
      mirroring=NoMirror
      resourceName=None
      resourceObject=null
      tiling=NoTile
      imageSize=<32 32>
      margin=<ltrb 0 0 0 0>
      outlineSettings {
       bUseBrushTransparency=false
       roundingType=FixedRadius
       width=1
       color {colorUseRule=UseColor_ColorTable, specifiedColor=<rgba 0 0 0 0>}
       cornerRadii=<4 4 4 4>
      }
      tintColor {colorUseRule=UseColor_Specified, specifiedColor=<rgba 0.0395 0.0395 0.0395 1>}
      uVRegion {bIsValid=false, max=<0 0>, min=<0 0>}
     }
     normalForeground {colorUseRule=UseColor_ColorTable, specifiedColor=<rgba 0 0 0 0>}
     normalPadding=<ltrb 0 0 0 0>
     pressed {
      bIsDynamicallyLoaded=false
      drawAs=RoundedBox
      imageType=NoImage
      mirroring=NoMirror
      resourceName=None
      resourceObject=null
      tiling=NoTile
      imageSize=<32 32>
      margin=<ltrb 0 0 0 0>
      outlineSettings {
       bUseBrushTransparency=false
       roundingType=FixedRadius
       width=1
       color {colorUseRule=UseColor_ColorTable, specifiedColor=<rgba 0 0 0 0>}
       cornerRadii=<4 4 4 4>
      }
      tintColor {colorUseRule=UseColor_Specified, specifiedColor=<rgba 0.0284 0.0284 0.0284 1>}
      uVRegion {bIsValid=false, max=<0 0>, min=<0 0>}
     }
     pressedForeground {colorUseRule=UseColor_ColorTable, specifiedColor=<rgba 0 0 0 0>}
     pressedPadding=<ltrb 0 0 0 0>
     pressedSlateSound {resourceObject=null}
    }}
   slot {class=HorizontalBoxSlot, properties {horizontalAlignment=HAlign_Right, verticalAlignment=VAlign_Center}}
  }
  Image_2 {
   widgetClassPath=&Image
   bIsVariable=true
   parent=ButtonDelete
   widget=Image_2
   properties {brush {
     bIsDynamicallyLoaded=false
     drawAs=Image
     imageType=FullColor
     mirroring=NoMirror
     resourceName=None
     resourceObject=/Editor_TodoList/Textures/T_Todo_Delete
     tiling=NoTile
     imageSize=<20 20>
     margin=<ltrb 0 0 0 0>
     outlineSettings {
      bUseBrushTransparency=false
      roundingType=HalfHeightRadius
      width=0
      color {colorUseRule=UseColor_Specified, specifiedColor=<rgba 0 0 0 0>}
      cornerRadii=<0 0 0 0>
     }
     tintColor {colorUseRule=UseColor_Specified, specifiedColor=<rgba 1 1 1 1>}
     uVRegion {bIsValid=false, max=<0 0>, min=<0 0>}
    }}
   slot {class=ButtonSlot, properties {padding=<ltrb 2 2 2 2>}}
  }
 ]
}
