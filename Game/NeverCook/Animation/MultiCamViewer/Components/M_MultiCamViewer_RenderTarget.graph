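# UI-domain material for the MultiCamViewer animation component, built to sample a camera render target, apparently for on-screen display.
# Its only node is a linear-color texture parameter named RenderTarget, defaulting to the face cam render target RT_MultiCamViewer_faceCam.
# The parameter lets material instances substitute other render targets, and the graph does no further processing on the sample.
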Material /Game/NeverCook/Animation/MultiCamViewer/Components/M_MultiCamViewer_RenderTarget
graph {
 properties {bAutomaticallySetUsageInEditor=false, materialDomain=MD_UI, usedShadingModels=Default Lit}
 nodes [TextureSampleParameter2D() {
   parameterName=RenderTarget
   samplerType=SAMPLERTYPE_LinearColor
   texture=&/RT_MultiCamViewer_faceCam
  }]
}
